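; Behavior tree task Blueprint that clears a configurable Blackboard key for an AI agent.
; On execute it clears the value of the TargetBBKey variable, then immediately finishes the task with success.
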(event EventReceiveExecute
  (params (OwnerActor :type Actor%20Object%20Reference))
  (AI|BehaviorTree|ClearBlackboardValue (Variables|Default|GetTargetBBKey))
  (AI|BehaviorTree|FinishExecute true))
