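; Behavior tree task that stops the AI agent's current movement.
; On execute it casts the owner actor to an AIController and calls StopMovement on it, finishing with success.
; If the cast fails, the task finishes with failure.
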(event EventReceiveExecute
  (params (OwnerActor :type Actor%20Object%20Reference))
  (bind _asaicontroller (Utilities|Casting|CastToAIController OwnerActor)
    (:then
      (AI|Navigation|StopMovement _asaicontroller)
      (AI|BehaviorTree|FinishExecute true))
    (:CastFailed
      (AI|BehaviorTree|FinishExecute false))))
