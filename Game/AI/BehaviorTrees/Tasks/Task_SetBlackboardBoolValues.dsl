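; Behavior tree task that writes up to five boolean Blackboard values in a single step.
; On execute, a sequence checks five update flags and, for each one that is set, writes that slot's configured value to its Blackboard key.
; The task then always finishes with success, regardless of which slots are updated.
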(event EventReceiveExecute
  (params (OwnerActor :type Actor%20Object%20Reference))
  (Utilities|FlowControl|Sequence
    (:then_0
      (if (Variables|Default|GetUpdate%201?)
        (AI|BehaviorTree|SetBlackboardValueasBool (Variables|Default|GetBoolKey1) (Variables|Default|GetBoolKey1Value))))
    (:then_1
      (if (Variables|Default|GetUpdate%202?)
        (AI|BehaviorTree|SetBlackboardValueasBool (Variables|Default|GetBoolKey2) (Variables|Default|GetBoolKey2Value))))
    (:then_2
      (if (Variables|Default|GetUpdate%203?)
        (AI|BehaviorTree|SetBlackboardValueasBool (Variables|Default|GetBoolKey3) (Variables|Default|GetBoolKey3Value))))
    (:then_3
      (if (Variables|Default|GetUpdate%204?)
        (AI|BehaviorTree|SetBlackboardValueasBool (Variables|Default|GetBoolKey4) (Variables|Default|GetBoolKey4Value))))
    (:then_4
      (if (Variables|Default|GetUpdate%205?)
        (AI|BehaviorTree|SetBlackboardValueasBool (Variables|Default|GetBoolKey5) (Variables|Default|GetBoolKey5Value))))
    (:then_5
      (AI|BehaviorTree|FinishExecute true))))
